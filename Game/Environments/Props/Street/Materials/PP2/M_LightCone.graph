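# Translucent light cone material for street props that fakes a light beam with a tinted emissive color and soft opacity.
# Opacity combines two slowly panning noise samples with an inverted vertical UV gradient and an inverted Fresnel that softens silhouette edges.
# Depth fade softens geometry intersections, and a pixel depth ramp fades the cone near the camera.
# The resulting material attributes pass through MF_CreativeMoveTool.
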Material /Game/Environments/Props/Street/Materials/PP2/M_LightCone
graph {
 properties {bUseMaterialAttributes=true, blendMode=BLEND_Translucent, pixelDepthOffsetMode=PDOM_Legacy, usedShadingModels=Default Lit}
 nodes [
  n1 = OneMinus(None=MaterialFunctionCall(UV Channel=TextureCoordinate()) {materialFunction=&LinearGradient}.VGradient)
  n2 = ScalarParameter() {defaultValue=0.0872, group=Light Cone, parameterName=Light Cone Opacity}
  n3 = TextureCoordinate() {vTiling=0.1}
  n4 = TextureCoordinate() {vTiling=0.05}
  n5 = Add(
   A=TextureSample(UVs=Panner(Coordinate=@n3) {speedX=0.005}) {texture=/Game/Global/Textures/Default/Noise/T_LightConeNoise}.G
   B=TextureSample(UVs=Panner(Coordinate=@n4) {speedX=0.002}) {texture=&T_LightConeNoise}.R
  )
  n6 = Divide(A=PixelDepth().R, B=ScalarParameter() {defaultValue=256, group=Light Cone, parameterName=Light Cone Near Depth Fade})
  n7 = DepthFade(
   Opacity=Multiply(A=Multiply(A=@n2, B=@n5), B=Clamp(None=@n6))
   FadeDistance=ScalarParameter() {defaultValue=64, group=Light Cone, parameterName=Light Cone Depth Fade}
  )
  n8 = VertexNormalWS()
  n9 = StaticBool()
  n10 = OneMinus(None=Abs(None=MaterialFunctionCall(Normal Vector=@n8, Clamp Fresnel Dot Product=@n9) {materialFunction=&Fresnel_Function}.Result))
  n11 = ScalarParameter() {defaultValue=2.2386, group=Light Cone, parameterName=Fresnel Power}
  n12 = Clamp(None=Multiply(A=Multiply(A=@n1, B=@n7), B=Power(Base=@n10, Exp=@n11)))
  n13 = ScalarParameter() {defaultValue=1, group=Light Cone, parameterName=MaskOpacityPower}
  MaterialFunctionCall(MA In=MakeMaterialAttributes!(
    BaseColor
    Metallic
    Specular
    Roughness
    Anisotropy
    EmissiveColor=VectorParameter() {group=Light Cone, parameterName=Emissive Color, defaultValue=<rgba 1.9479 2 0.7793 0.2569>}.RGB
    Opacity=Power(Base=@n12, Exp=@n13)
    OpacityMask
    Normal
    Tangent
    WorldPositionOffset
    SubsurfaceColor
    ClearCoat
    ClearCoatRoughness
    AmbientOcclusion
    Refraction
    CustomizedUVs_0
    CustomizedUVs_1
    CustomizedUVs_2
    CustomizedUVs_3
    CustomizedUVs_4
    CustomizedUVs_5
    CustomizedUVs_6
    CustomizedUVs_7
    PixelDepthOffset
    ShadingModel
    Displacement
   )) {materialFunction=&MF_CreativeMoveTool}
 ]
}
